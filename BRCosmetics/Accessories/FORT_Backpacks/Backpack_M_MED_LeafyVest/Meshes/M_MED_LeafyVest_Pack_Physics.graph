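# Physics asset for the LeafyVest backpack cosmetic, defining capsule bodies and joints for the pack and three short chains (left, middle, right).
# One large capsule covers the pack, and each chain has a tiny base body jointed to it with free swing and twist.
# Each dynamic chain tip hangs off its base with zero twist and limited swing, with the middle tip restricted to a narrow single-axis swing.
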PhysicsAsset /BRCosmetics/Accessories/FORT_Backpacks/Backpack_M_MED_LeafyVest/Meshes/M_MED_LeafyVest_Pack_Physics
bodies {
 base_chain_l [{shapeType=Capsule, length=0.505, radius=0.505, shapeName=base_chain_l_capsule, rotation=<rot 0 -90 0>}]
 base_chain_m [{shapeType=Capsule, length=0.505, radius=0.505, shapeName=base_chain_m_capsule, rotation=<rot 0 -90 0>}]
 base_chain_r [{shapeType=Capsule, length=0.505, radius=0.505, shapeName=base_chain_r_capsule, rotation=<rot 0 -90 0>}]
 dyn_chain_l [{shapeType=Capsule, length=4, radius=2, shapeName=dyn_chain_l_capsule, center=<3.4208 0 0>, rotation=<rot 90 0 0>}]
 dyn_chain_m [{shapeType=Capsule, length=4, radius=2, shapeName=dyn_chain_m_capsule, center=<6.3588 0 0>, rotation=<rot 90 0 0>}]
 dyn_chain_r [{shapeType=Capsule, length=4, radius=2, shapeName=dyn_chain_r_capsule, center=<3.4208 0 0>, rotation=<rot 90 0 0>}]
 pack [{shapeType=Capsule, length=40.5558, radius=14.4041, shapeName=pack_capsule, center=<0 12.4227 -18.4992>, rotation=<rot 0 90 90>}]
}
constraints [
 {bone1Name=dyn_chain_l, bone2Name=base_chain_l, swing1LimitDegrees=10, swing2LimitDegrees=45, twistLimitDegrees=0}
 {bone1Name=dyn_chain_m, bone2Name=base_chain_m, swing1LimitDegrees=0, swing2LimitDegrees=18, twistLimitDegrees=0}
 {bone1Name=dyn_chain_r, bone2Name=base_chain_r, swing1LimitDegrees=10, swing2LimitDegrees=45, twistLimitDegrees=0}
 {
  bone1Name=base_chain_l
  bone2Name=pack
  swing1LimitDegrees=45
  swing1Motion=Free
  swing2LimitDegrees=45
  swing2Motion=Free
  twistLimitDegrees=45
  twistMotion=Free
 }
 {
  bone1Name=base_chain_m
  bone2Name=pack
  swing1LimitDegrees=45
  swing1Motion=Free
  swing2LimitDegrees=45
  swing2Motion=Free
  twistLimitDegrees=45
  twistMotion=Free
 }
 {
  bone1Name=base_chain_r
  bone2Name=pack
  swing1LimitDegrees=45
  swing1Motion=Free
  swing2LimitDegrees=45
  swing2Motion=Free
  twistLimitDegrees=45
  twistMotion=Free
 }
]
