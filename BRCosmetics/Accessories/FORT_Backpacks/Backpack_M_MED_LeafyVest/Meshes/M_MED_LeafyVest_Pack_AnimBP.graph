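# Animation Blueprint for the LeafyVest backpack that adds physics motion and toggles a material element.
# Updates cache the parent body's spine socket transforms and compute gravity, spine-twist and movement-state values.
# The AnimGraph applies world-space spine replacements, rigid-body simulation with gravity override, then additive chain-base offsets, one bool-blended.
# Weapon, customization and skydive events set "Hide Element 01" to 1 when a LeafyVest-tagged pickaxe is equipped outside skydiving.
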AnimBlueprint /BRCosmetics/Accessories/FORT_Backpacks/Backpack_M_MED_LeafyVest/Meshes/M_MED_LeafyVest_Pack_AnimBP
target_skeleton=&/M_MED_LeafyVest_Pack_Skeleton
graphs [
 AnimGraph {nodes [
   n1 = "Transform (Modify) Bone - Bone: spine_01"(ComponentPose=Local To Component(LocalPose=Input Pose()), Alpha=1, Translation="0,0,0", Rotation="0,0,0", Scale="1,1,1") {
    rotationMode=BMM_Replace
    rotationSpace=BCS_WorldSpace
    scaleSpace=BCS_BoneSpace
    translationMode=BMM_Replace
    translationSpace=BCS_WorldSpace
    alphaBoolBlend {blendOption=Linear}
    alphaScaleBiasClamp {inRange {max=0, min=-40}, outRange {max=0, min=1}}
    boneToModify {boneName=spine_01}
   }
   n2 = "Transform (Modify) Bone - Bone: spine_02"(ComponentPose=@n1, Alpha=1, Translation="0,0,0", Rotation="0,0,0", Scale="1,1,1") {
    rotationMode=BMM_Replace
    rotationSpace=BCS_WorldSpace
    scaleSpace=BCS_BoneSpace
    translationMode=BMM_Replace
    translationSpace=BCS_WorldSpace
    alphaBoolBlend {blendOption=Linear}
    alphaScaleBiasClamp {inRange {max=0, min=-40}, outRange {max=0, min=1}}
    boneToModify {boneName=spine_02}
   }
   n3 = "Transform (Modify) Bone - Bone: spine_03"(ComponentPose=@n2, Alpha=1, Translation="0,0,0", Rotation="0,0,0", Scale="1,1,1") {
    rotationMode=BMM_Replace
    rotationSpace=BCS_WorldSpace
    scaleSpace=BCS_BoneSpace
    translationMode=BMM_Replace
    translationSpace=BCS_WorldSpace
    alphaBoolBlend {blendOption=Linear}
    alphaScaleBiasClamp {inRange {max=0, min=-40}, outRange {max=0, min=1}}
    boneToModify {boneName=spine_03}
   }
   n4 = "Transform (Modify) Bone - Bone: spine_04"(ComponentPose=@n3, Alpha=1, Translation="0,0,0", Rotation="0,0,0", Scale="1,1,1") {
    rotationMode=BMM_Replace
    rotationSpace=BCS_WorldSpace
    scaleSpace=BCS_BoneSpace
    translationMode=BMM_Replace
    translationSpace=BCS_WorldSpace
    alphaBoolBlend {blendOption=Linear}
    alphaScaleBiasClamp {inRange {max=0, min=-40}, outRange {max=0, min=1}}
    boneToModify {boneName=spine_04}
   }
   n5 = RigidBody!(ComponentPose=@n4, Alpha, OverrideWorldGravity:Vector="0,0,0", ExternalForce) -> (Pose) {
    bOverrideWorldGravity=true
    overridePhysicsAsset=&/M_MED_LeafyVest_Pack_Physics
    alphaBoolBlend {blendOption=Linear}
    baseBoneRef {boneName=pack}
    componentAppliedLinearAccClamp=<3000 3000 3000>
    componentLinearAccScale=<0.3 0.3 0.3>
    componentLinearVelScale=<0.3 0.3 0.3>
    simSpaceSettings {
     maxAngularAcceleration=800
     maxAngularVelocity=2.5
     maxLinearAcceleration=6
     maxLinearVelocity=450
     velocityScaleZ=0
     worldAlpha=0.4
     externalLinearDragV=<0.2 0.2>
    }
   }
   n6 = "Transform (Modify) Bone - Bone: base_chain_l"(ComponentPose=@n5, Alpha=1, Translation="0,0,0", Rotation="35,0,0", Scale="1,1,1") {
    rotationMode=BMM_Additive
    rotationSpace=BCS_BoneSpace
    alphaBoolBlend {blendOption=Linear}
    boneToModify {boneName=base_chain_l}
    rotation {pitch=35}
   }
   n7 = "Transform (Modify) Bone - Bone: base_chain_m"(ComponentPose=@n6, Alpha=1, Translation="5,0,5.662", Rotation="-0.000,19.901,-0.000", Scale="1,1,1") {
    rotationMode=BMM_Additive
    rotationSpace=BCS_BoneSpace
    translationMode=BMM_Additive
    translationSpace=BCS_BoneSpace
    alphaBoolBlend {blendOption=Linear}
    boneToModify {boneName=base_chain_m}
    rotation=<rot -0.0001 19.901 0>
    translation {x=5, z=5.662}
   }
   n8 = "Transform (Modify) Bone - Bone: base_chain_r"(ComponentPose=@n7, Alpha=1, Translation="0,0,0", Rotation="35,0,0", Scale="1,1,1") {
    rotationMode=BMM_Additive
    rotationSpace=BCS_BoneSpace
    alphaBoolBlend {blendOption=Linear}
    boneToModify {boneName=base_chain_r}
    rotation {pitch=35}
   }
   n9 = "Transform (Modify) Bone - Bone: base_chain_m"!(ComponentPose=@n8, bAlphaBoolEnabled:bool=True, Translation="-7,0,0", Rotation="0,0,0", Scale="1,1,1") -> (Pose) {
    alphaInputType=Bool
    rotationMode=BMM_Additive
    rotationSpace=BCS_BoneSpace
    translationMode=BMM_Additive
    translationSpace=BCS_BoneSpace
    alphaBoolBlend {blendInTime=0.6, blendOption=Linear, blendOutTime=0.2}
    boneToModify {boneName=base_chain_m}
    translation {x=-7}
   }
   Output Pose(Result=Component To Local(ComponentPose=@n9)) {layerGroup=None}
  ]}
 EventGraph {nodes [
   Event BlueprintUpdateAnimation()
   > n1 = Sequence()
   > n2 = Gravity Override State Function DEPRECATED(CharacterPartAnimInstance=Self-Reference(), GameTimeInSeconds=GetGameTimeInSeconds(), GravityOverrideParams=Get Gravity Override Params())
   > Set GravityOverride(GravityOverride=@n2)
   > n3 = Sequence()
   n4 = GetPartSkeletalMeshComponent()
   > n5 = Set Skel Mesh(Skel Mesh=@n4)
   > Set spineRot(spineRot:double=GetAlphaFromReferenceTransformDelta(SkeletalMeshComponent=@n5, BoneName=spine_04, BaseName=spine_01, InputRange=(X=0,Y=-22.1), TwistAxis="0,0,1")) -> (Output_Get:double)
   Is Valid(exec=@n3.then_1, InputObject=@n4)
   n6 = GetAttachParent!(self:Scene Component Object Reference=@n4) -> (ReturnValue:Scene Component Object Reference)
   > Is Valid(InputObject=@n6)
   > Set Body_Mesh(Body_Mesh:Scene Component Object Reference=@n6) -> (Output_Get:Scene Component Object Reference)
   > n7 = Sequence!(execute) -> (then_0, then_1, then_2:Exec, then_3:Exec)
   n8 = Get Body_Mesh() -> (Body_Mesh:Scene Component Object Reference)
   n9 = BreakTransform(InTransform=GetSocketTransform(self=@n8, InSocketName=spine_01, TransformSpace=RTS_World))
   > Set spine_01_location(spine_01_location:Vector=@n9) -> (Output_Get:Vector)
   > Set spine_01_rotation(spine_01_rotation:Rotator=@n9.Rotation) -> (Output_Get:Rotator)
   n10 = BreakTransform(InTransform=GetSocketTransform(self=@n8, InSocketName=spine_02, TransformSpace=RTS_World))
   Set spine_02_location(execute=@n7.then_1, spine_02_location:Vector=@n10) -> (Output_Get:Vector)
   > Set spine_02_rotation(spine_02_rotation:Rotator=@n10.Rotation) -> (Output_Get:Rotator)
   n11 = BreakTransform(InTransform=GetSocketTransform(self=@n8, InSocketName=spine_04, TransformSpace=RTS_World))
   Set spine_04_location(execute=@n7.then_3, spine_04_location:Vector=@n11) -> (Output_Get:Vector)
   > Set spine_04_rotation(spine_04_rotation:Rotator=@n11.Rotation) -> (Output_Get:Rotator)
   n12 = BreakTransform(InTransform=GetSocketTransform(self=@n8, InSocketName=spine_03, TransformSpace=RTS_World))
   Set spine_03_location(execute=@n7.then_2, spine_03_location:Vector=@n12) -> (Output_Get:Vector)
   > Set spine_03_rotation(spine_03_rotation:Rotator=@n12.Rotation) -> (Output_Get:Rotator)
   n13 = Get bIsSkyDiving()
   n14 = Get bIsMoving2D()
   n15 = Get bIsFalling()
   n16 = Get bIsJumping()
   Set ChainMover(
    execute=@n1.then_1
    ChainMover:bool=AND Boolean!(A=OR Boolean(A=@n16, B=@n15), B=NOT Boolean(A=@n14), C:bool=NOT Boolean(A=@n13)) -> (ReturnValue)
   ) -> (Output_Get:bool)
   > n17 = Branch(Condition=Get bIsSkyDiving())
   > Set skydiveFloat(skydiveFloat:double=0.7) -> (Output_Get:double)
   Set skydiveFloat(execute=@n17.else, skydiveFloat=1)
   n18 = Event OnWeaponChanged()
   n19 = Get Dynamic Material Instance(execute=[@n18.then, @n25.then], SkeletalMeshComponent=GetPartSkeletalMeshComponent())
   n20 = PickaxeItemDefinitionHasAnyMetaTag(Tags=Make Array("[0]"=(TagName="Cosmetics.MetaTags.LeafyVest")), bRequiresEquipped=true)
   > n21 = Branch(Condition=@n20)
   > n22 = Branch(Condition=Get bIsSkyDiving())
   SetScalarParameterValue(execute=[@n23.then, @n22.then, @n24.else, @n21.else], self=@n19, ParameterName=Hide Element 01, Value=0)
   SetScalarParameterValue(execute=[@n22.else, @n24.then], self=@n19, ParameterName=Hide Element 01, Value=1)
   n23 = Event OnBeginSkydiving()
   Event OnEndSkydiving()
   > n24 = Branch(Condition=@n20)
   n25 = Event OnCharacterCustomizationLoaded()
   TryGetPawnOwner()
  ]}
]
variables [
 Body_Mesh {type=SceneComponent, default=null}
 body_root_location {type=Vector}
 body_root_rotation {type=Rotator}
 BodyMesh {type=SkeletalMeshComponentBudgeted}
 ChainMover {type=boolean}
 Gravity Override Params {type=GravityOverrideParamsStruct}
 GravityOverride {type=Vector}
 pelvis_location {type=Vector}
 pelvis_rotation {type=Rotator}
 Skel Mesh {type=SkeletalMeshComponent}
 skydiveFloat {type=number}
 spine_01_location {type=Vector}
 spine_01_rotation {type=Rotator}
 spine_02_location {type=Vector}
 spine_02_rotation {type=Rotator}
 spine_03_location {type=Vector}
 spine_03_rotation {type=Rotator}
 spine_04_location {type=Vector}
 spine_04_rotation {type=Rotator}
 spine_05_location {type=Vector}
 spine_05_rotation {type=Rotator}
 spineRot {type=number}
]
